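# Masked, lit material for the Vertigo character's tentacle FX, usable on Niagara ribbons, meshes and skeletal meshes.
# Panned ribbon and noise masks, an edge gradient and optional breakup form a particle-alpha-scaled mask, clipped via dynamic parameter.
# A cylinder normal blends with panning swirl noise, while fresnel mixes two base colors tinted by particle color.
# Emissive follows sun direction and time of day, with a simple-color override; temporal dither scales a depth threshold.
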Material /Game/Characters/Player/Male/Large/Bodies/M_LRG_Vertigo/FX/MT_VertigoTentacle_Ribbon_pbr
graph {
 properties {
  bUsedWithMeshDeformer=true
  bUsedWithNiagaraMeshParticles=true
  bUsedWithNiagaraRibbons=true
  bUsedWithSkeletalMesh=true
  blendMode=BLEND_Masked
  pixelDepthOffsetMode=PDOM_Legacy
  usedShadingModels=Default Lit
 }
 nodes [
  n1 = VectorParameter() {group=BreakUp Alpha Mask, parameterName=Ribbon BreakUp Mask Panning, defaultValue=<rgba -1 0 0 0>}
  n2 = DynamicParameter() {defaultValue=<rgba 0.5 1 1 1>, paramNames [AlphaClipAmount, RandomOffset, MaskScrollSpeed, EmissiveIntensity]}
  AppendVector(A=Multiply(A=ComponentMask(None=@n1.RGB) {r=true}, B=@n2.MaskScrollSpeed), B=ComponentMask(None=@n1.RGB) {g=true})
  n3 = VectorParameter() {group=BreakUp Alpha Mask, parameterName=Ribbon BreakUp Mask UVScale, defaultValue=<rgba 2 1 0 1>}
  AppendVector(A=ComponentMask(None=@n3.RGB) {r=true}, B=ComponentMask(None=@n3.RGB) {g=true})
  Constant() {r=1}
  n4 = TextureCoordinate()
  n5 = ComponentMask(None=@n4) {r=true}
  n6 = ScalarParameter() {defaultValue=5, group=Normal, parameterName=Normal Tile U}
  n7 = ComponentMask(None=@n4) {g=true}
  n8 = ScalarParameter() {defaultValue=0.7, group=Normal, parameterName=Normal Tile V}
  n9 = Panner(
   Coordinate=AppendVector(A=Multiply(A=@n5, B=@n6), B=Multiply(A=@n7, B=@n8))
   Time=Time() {bOverride_Period=true, period=64}
   Speed=AppendVector(
    A=ScalarParameter() {defaultValue=1, group=Normal, parameterName=Normal Pan Speed U}
    B=ScalarParameter() {group=Normal, parameterName=Normal Pan Speed V}
   )
  ) {speedX=-1, speedY=0.1}
  MaterialFunctionCall(
   BaseNormal=TextureSample() {samplerType=SAMPLERTYPE_Normal, texture=&/T_CylinderNormal}.RGB
   AdditionalNormal=Multiply(
    A=TextureSample(UVs=@n9) {samplerType=SAMPLERTYPE_Normal, texture=/Game/Global/Textures/Default/Noise/T_SwirlNoise_N}.RGB
    B=VectorParameter() {group=Normal, parameterName=NormalStrength, defaultValue=<rgba 0.4 0.4 1 1>}.RGB
   )
  ) {materialFunction=&BlendAngleCorrectedNormals}
  n10 = CollectionParameter() {
   collection=&FortniteMaterialParameters
   parameterName=SunAndMoonAtmosphereDirectionalVector
  }
  n11 = DotProduct(A=ComponentMask(None=@n10) {b=true, g=true, r=true}, B=PixelNormalWS())
  n12 = Multiply(
   A=MaterialFunctionCall(
    Color_Morning=VectorParameter() {group=TOD, parameterName=TOD_Morning, defaultValue=<rgba 0.3675 0.1266 0.0895 1>}.RGB
    Color_Day=VectorParameter() {group=TOD, parameterName=TOD_Afternoon, defaultValue=<rgba 1 1 1 1>}.RGB
    Color_Evening=VectorParameter() {group=TOD, parameterName=TOD_Evening, defaultValue=<rgba 0.3316 0.1538 0.1088 1>}.RGB
    Color_Night=VectorParameter() {group=TOD, parameterName=TOD_Night, defaultValue=<rgba 0.0874 0.1753 0.1753 1>}.RGB
   ) {materialFunction=&Color_Time_of_Day}.Result
   B=LinearInterpolate(
    A=VectorParameter() {parameterName=emissiveColorAgainstSun, defaultValue=<rgba 0.0041 0.0058 0.0133 0>}.RGB
    B=VectorParameter() {parameterName=EmissiveColorInSun}.RGB
    Alpha=Clamp(None=@n11)
   )
  )
  n13 = DynamicParameter() {defaultValue=<rgba 0.5 1 1 1>, paramNames [AlphaClipAmount, RandomOffset, MaskScrollSpeed, EmissiveIntensity]}
  MaterialFunctionCall(Emissive=LinearInterpolate(
    A=Multiply(A=@n12, B=@n13.EmissiveIntensity)
    B=VectorParameter() {parameterName=Simple_EmissiveColor, defaultValue=<rgba 1 1 1 1>}.RGB
    Alpha=ScalarParameter() {parameterName=Only_Use_simple_EmissiveColor}
   )) {materialFunction=&MF_FX_EmissiveAdjustments}
  n14 = ScalarParameter() {defaultValue=1, parameterName=FresnelMult}
  n15 = ScalarParameter() {defaultValue=1, parameterName=FresnelExp}
  n16 = Transform(None=TextureSample() {samplerType=SAMPLERTYPE_Normal, texture=&T_CylinderNormal}.RGB)
  Multiply(
   A=LinearInterpolate(
    A=VectorParameter() {parameterName=baseColorOuter, defaultValue=<rgba 0.0425 0.0725 0.14 0>}.RGB
    B=VectorParameter() {parameterName=baseColorInner, defaultValue=<rgba 0.0213 0.0239 0.035 0>}.RGB
    Alpha=MaterialFunctionCall(Brightness=@n14, Exponent=@n15, Vertex / Pixel Normal=@n16) {materialFunction=&MF_Fresnel_FX}.Fresnel
   )
   B=ParticleColor().RGB
  )
  Multiply(
   A=ScalarParameter() {defaultValue=5, parameterName=DitherDepthThershold}
   B=MaterialFunctionCall() {materialFunction=&DitherTemporalAA}.Result
  )
  n17 = VectorParameter() {group=Alpha Mask, parameterName=Ribbon Mask UV Scale, defaultValue=<rgba 5 1 0 1>}
  n18 = Panner(
   Coordinate=Multiply(A=TextureCoordinate(), B=ComponentMask(None=@n17.RGB) {g=true, r=true})
   Time=Time() {bOverride_Period=true, period=64}
   Speed=ComponentMask(None=VectorParameter() {group=Alpha Mask, parameterName=Ribbon Mask Panner Speed, defaultValue=<rgba 0.75 0 0 0>}.RGB) {g=true, r=true}
  )
  n19 = TextureCoordinate()
  n20 = ComponentMask(None=VectorParameter() {group=Alpha Mask, parameterName=Noise Mask UV Scale, defaultValue=<rgba 1 0.2 0 1>}.RGB) {g=true, r=true}
  n21 = VectorParameter() {group=Alpha Mask, parameterName=Noise Mask Panner Speed, defaultValue=<rgba 0.05 0 0 0>}
  n22 = TextureSampleParameter2D(UVs=Panner(Coordinate=Multiply(A=@n19, B=@n20), Time=Time() {bOverride_Period=true, period=64}, Speed=ComponentMask(None=@n21.RGB) {g=true, r=true})) {
   group=Alpha Mask
   parameterName=Noise Mask
   samplerSource=SSM_Wrap_WorldGroupSettings
   texture=/Game/Global/Textures/Default/Caustics/T_Fractal_01
  }
  n23 = ScalarParameter() {defaultValue=0.1, group=Alpha Mask, parameterName=Noise Mask Contrast}
  n24 = TextureSampleParameter2D() {
   group=Alpha Mask
   parameterName=Edge Alpha Mask
   samplerType=SAMPLERTYPE_LinearColor
   texture=/Game/Global/Textures/Default/Masks_and_Gradients/T_VertGradient
  }
  n25 = Add(
   A=Multiply(
    A=TextureSampleParameter2D(UVs=@n18) {group=Alpha Mask, parameterName=Ribbon Mask, texture=/Game/Effects/Fort_Effects/Textures/Smoke/T_Smoke_Trail_Soft_Chromatic}.RGB
    B=MaterialFunctionCall(In=@n22.RGB, Contrast=@n23) {materialFunction=&CheapContrast}.Result
   )
   B=MaterialFunctionCall(In=@n24.RGB, Contrast=ScalarParameter() {group=Alpha Mask, parameterName=Edge Mask Contrast}) {materialFunction=&CheapContrast}.Result
  )
  n26 = TextureCoordinate()
  n27 = ComponentMask(None=@n3.RGB) {g=true, r=true}
  n28 = Time() {bOverride_Period=true, period=64}
  n29 = TextureSample(UVs=Panner(Coordinate=Multiply(A=@n26, B=@n27), Time=Add(A=@n28, B=@n2.RandomOffset), Speed=ComponentMask(None=@n1.RGB) {g=true, r=true})) {texture=/Game/Weapons/FORT_Rifles/Effects/Textures/FX_T_Gradient_Soundwave_01}
  n30 = Multiply(
   A=StaticSwitchParameter(True=Multiply(A=@n25, B=@n29.RGB), False=@n25) {group=BreakUp Alpha Mask, parameterName=Use Mask BreakUp}
   B=ParticleColor().A
  )
  n31 = ScalarParameter() {defaultValue=0.05, group=Alpha Mask, parameterName=ClipRange}
  n32 = ScalarParameter() {defaultValue=0.5, group=Alpha Mask, parameterName=ClipAmnt}
  n33 = DynamicParameter() {defaultValue=<rgba 0.5 1 1 1>, paramNames [AlphaClipAmount, RandomOffset, MaskScrollSpeed, EmissiveIntensity]}
  n34 = LinearInterpolate(A=Multiply(A=@n31) {constB=-1}, Alpha=OneMinus(None=Multiply(A=@n32, B=@n33.AlphaClipAmount)) {desc=Inverse Alpha})
  Saturate(None=Divide(A=Subtract(A=@n30, B=@n34), B=@n31))
  ScalarParameter() {defaultValue=0.2, parameterName=Specular}
  ScalarParameter() {defaultValue=0.4, parameterName=Roughness}
  ScalarParameter() {defaultValue=0.4, parameterName=Metalic}
 ]
}
