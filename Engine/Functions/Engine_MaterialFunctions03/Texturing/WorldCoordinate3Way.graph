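# WorldCoordinate3Way is a library material function that triplanar-projects textures onto surfaces using world position.
# It samples XY, XZ and YZ planes at separate scales, with a static bool to reuse the XY texture for all three.
# The transformed absolute normal, shaped by exponent and multiplier, drives two nested lerps: side against side, then top over sides.
# Outputs provide the blend and each planar sample.
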MaterialFunction /Engine/Functions/Engine_MaterialFunctions03/Texturing/WorldCoordinate3Way
graph {
 properties {bExposeToLibrary=true, description=Takes textures and projects them in world-space onto the surface., libraryCategoriesText [Misc,Texturing]}
 nodes [
  n1 = FunctionInput(Preview=WorldPosition() {worldPositionShaderOffset=WPT_ExcludeAllShaderOffsets}.XYZ) {bUsePreviewValueAsDefault=true, description=Optional input for altered World position coordinates, inputName=WorldPosition, sortPriority=10}
  n2 = FunctionInput() {
   bUsePreviewValueAsDefault=true
   description=Scale of XZ projected texture in world units
   inputName=XZ Scale
   inputType=FunctionInput_Scalar
   sortPriority=2
   previewValue=<512 0 0 1>
  }
  n3 = TextureObject() {texture=/Engine/EngineMaterials/DefaultDiffuse}
  n4 = FunctionInput(Preview=@n3) {bUsePreviewValueAsDefault=true, inputName=XY Texture, inputType=FunctionInput_Texture2D, sortPriority=-1}
  n5 = FunctionInput(Preview=StaticBool()) {
   bUsePreviewValueAsDefault=true
   description="When true, only uses the texture in the XY entry for all three projections."
   inputName=Single Texture
   inputType=FunctionInput_StaticBool
   sortPriority=7
  }
  n6 = TextureSample(
   UVs=Divide(A=ComponentMask(None=@n1) {b=true, r=true}, B=@n2)
   Tex=StaticSwitch(
    True=@n4
    False=FunctionInput(Preview=@n3) {bUsePreviewValueAsDefault=true, inputName=XZ Texture, inputType=FunctionInput_Texture2D}
    Value=@n5
   )
  ) {texture=/Engine/EngineResources/DefaultTexture}
  n7 = FunctionInput() {
   bUsePreviewValueAsDefault=true
   description=Scale of YZ projected texture in world units
   inputName=YZ Scale
   inputType=FunctionInput_Scalar
   sortPriority=2
   previewValue=<512 0 0 1>
  }
  n8 = TextureSample(
   UVs=Divide(A=ComponentMask(None=@n1) {b=true, g=true}, B=@n7)
   Tex=StaticSwitch(
    True=@n4
    False=FunctionInput(Preview=@n3) {bUsePreviewValueAsDefault=true, inputName=YZ Texture, inputType=FunctionInput_Texture2D}
    Value=@n5
   )
  ) {texture=&DefaultTexture}
  n9 = Constant3Vector() {constant=<rgba 0 0 1 0>}
  n10 = Abs(None=Transform(None=FunctionInput(Preview=@n9) {bUsePreviewValueAsDefault=true, description=Brings the normal into the blend., inputName=Normal, sortPriority=8}))
  n11 = FunctionInput() {
   bUsePreviewValueAsDefault=true
   description=Sets the power for the blend between the sides.
   inputName=Blend Exponent Y Z
   inputType=FunctionInput_Scalar
   sortPriority=3
   previewValue=<1.5 0 0 1>
  }
  n12 = Power(Base=@n10, Exp=@n11)
  n13 = FunctionInput() {
   bUsePreviewValueAsDefault=true
   description=Multiplies the blend between the side projections
   inputName=Blend Mult Y Z
   inputType=FunctionInput_Scalar
   sortPriority=4
   previewValue=<1.5 0 0 1>
  }
  n14 = Clamp(None=Multiply(A=ComponentMask(None=@n12) {r=true}, B=@n13))
  n15 = FunctionInput() {
   bUsePreviewValueAsDefault=true
   description=Scale of XY projected texture in world units
   inputName=XY Scale
   inputType=FunctionInput_Scalar
   sortPriority=1
   previewValue=<512 0 0 1>
  }
  n16 = TextureSample(UVs=Divide(A=ComponentMask(None=@n1) {g=true, r=true}, B=@n15), Tex=@n4) {texture=&DefaultTexture}
  n17 = FunctionInput() {
   bUsePreviewValueAsDefault=true
   description=Sets the power for the blend between the top and the sides.
   inputName=Blend Exponent X
   inputType=FunctionInput_Scalar
   sortPriority=5
   previewValue=<1.5 0 0 1>
  }
  n18 = FunctionInput() {
   bUsePreviewValueAsDefault=true
   description=Multiplies the blend between the top and the sides
   inputName=Blend Mult X
   inputType=FunctionInput_Scalar
   sortPriority=6
   previewValue=<1.5 0 0 1>
  }
  n19 = Multiply(A=ComponentMask(None=Power(Base=@n10, Exp=@n17)) {b=true}, B=@n18)
  FunctionOutput(None=LinearInterpolate(A=LinearInterpolate(A=@n6.RGB, B=@n8.RGB, Alpha=@n14), B=@n16.RGB, Alpha=Clamp(None=@n19))) {description=3 way blended texturing, outputName=XYZ Output}
  FunctionOutput(None=@n16.RGB) {description=XY Plane texturing, outputName=XY, sortPriority=1}
  FunctionOutput(None=@n6.RGB) {description=XZ Plane Texturing, outputName=XZ, sortPriority=2}
  FunctionOutput(None=@n8.RGB) {description=YZ Plane Texturing, outputName=YZ, sortPriority=3}
 ]
}
