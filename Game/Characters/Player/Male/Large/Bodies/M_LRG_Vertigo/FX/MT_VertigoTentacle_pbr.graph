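# Lit PBR material for the Vertigo character's tentacle FX, usable on skeletal meshes and Niagara meshes or ribbons.
# Base color blends two Fresnel colors with particle color and optional noise-masked NitroFlow tint, alongside a panning swirl normal.
# Emissive mixes sun-facing colors, optionally toon-stepped, with time-of-day tint.
# Animated world-triplanar or UV noise drives normal-based vertex offset, with dynamic parameters scaling thickness, offset and emissive.
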Material /Game/Characters/Player/Male/Large/Bodies/M_LRG_Vertigo/FX/MT_VertigoTentacle_pbr
graph {
 properties {
  bUsedWithMeshDeformer=true
  bUsedWithNiagaraMeshParticles=true
  bUsedWithNiagaraRibbons=true
  bUsedWithSkeletalMesh=true
  pixelDepthOffsetMode=PDOM_Legacy
  usedShadingModels=Default Lit
 }
 nodes [
  If()
  n1 = ScalarParameter() {defaultValue=1, parameterName=FresnelMult}
  n2 = ScalarParameter() {defaultValue=0.5, parameterName=FresnelExp}
  n3 = Multiply(
   A=LinearInterpolate(
    A=VectorParameter() {parameterName=baseColorOuter, defaultValue=<rgba 0.0425 0.0725 0.14 0>}.RGB
    B=VectorParameter() {parameterName=baseColorInner, defaultValue=<rgba 0.0213 0.0239 0.035 0>}.RGB
    Alpha=MaterialFunctionCall(Brightness=@n1, Exponent=@n2) {materialFunction=&MF_Fresnel_FX}.Fresnel
   )
   B=ParticleColor().RGB
  )
  n4 = LinearInterpolate(
   A=Time() {bOverride_Period=true, period=64}
   B=Time() {period=64}
   Alpha=ScalarParameter() {group=NitroFlow_Options, parameterName=UseLinearTime}
  )
  n5 = AppendVector(
   A=ScalarParameter() {group=Normal, parameterName=Normal Pan Speed U}
   B=ScalarParameter() {defaultValue=1, group=Normal, parameterName=Normal Pan Speed V}
  )
  n6 = Panner(Coordinate=TextureCoordinate(), Time=@n4, Speed=@n5)
  LinearInterpolate(
   A=@n3
   B=LinearInterpolate(
    A=@n3
    B=VectorParameter() {group=NitroFlow_Options, parameterName=NitroFlow_ColorA, defaultValue=<rgba 0 0.1875 0.1037 1>}.RGB
    Alpha=TextureSample(UVs=@n6) {samplerType=SAMPLERTYPE_Grayscale, texture=/Game/UI/Frontend/ItemShop/Art/Rumble/LowResBlurredNoise_Rumble}.RGB
   )
   Alpha=ScalarParameter() {group=NitroFlow_Options, parameterName=UseNitroFlow_Colors}
  )
  n7 = MaterialFunctionCall(
   Color_Morning=VectorParameter() {group=TOD, parameterName=TOD_Morning, defaultValue=<rgba 0.3675 0.1266 0.0895 1>}.RGB
   Color_Day=VectorParameter() {group=TOD, parameterName=TOD_Afternoon, defaultValue=<rgba 1 1 1 1>}.RGB
   Color_Evening=VectorParameter() {group=TOD, parameterName=TOD_Evening, defaultValue=<rgba 0.3316 0.1538 0.1088 1>}.RGB
   Color_Night=VectorParameter() {group=TOD, parameterName=TOD_Night, defaultValue=<rgba 0.0874 0.1753 0.1753 1>}.RGB
  ) {materialFunction=&Color_Time_of_Day}
  n8 = CollectionParameter() {
   collection=&FortniteMaterialParameters
   parameterName=SunAndMoonAtmosphereDirectionalVector
  }
  n9 = Clamp(None=DotProduct(A=ComponentMask(None=@n8) {b=true, g=true, r=true}, B=PixelNormalWS()))
  n10 = LinearInterpolate(
   A=VectorParameter() {parameterName=emissiveColorAgainstSun, defaultValue=<rgba 0.0103 0.0145 0.033 0>}.RGB
   B=VectorParameter() {parameterName=EmissiveColorInSun}.RGB
   Alpha=LinearInterpolate(A=@n9, B=Ceil(None=@n9), Alpha=ScalarParameter() {parameterName=UseToon})
  )
  MaterialFunctionCall(Emissive=Multiply(A=Multiply(A=@n7.Result, B=@n10), B=DynamicParameter() {paramNames [Tendril Thickness, WPO Amount, Emissive, NA]}.Emissive)) {materialFunction=&MF_FX_EmissiveAdjustments}
  n11 = TextureCoordinate()
  n12 = AppendVector(
   A=Multiply(A=ComponentMask(None=@n11) {r=true}, B=ScalarParameter() {defaultValue=0.7, group=Normal, parameterName=Normal Tile U})
   B=Multiply(A=ComponentMask(None=@n11) {g=true}, B=ScalarParameter() {defaultValue=8, group=Normal, parameterName=Normal Tile V})
  )
  Multiply(
   A=TextureSample(UVs=Panner(Coordinate=@n12, Time=@n4, Speed=@n5) {speedX=-1, speedY=0.1}) {samplerType=SAMPLERTYPE_Normal, texture=/Game/Global/Textures/Default/Noise/T_SwirlNoise_N}.RGB
   B=VectorParameter() {group=Normal, parameterName=NormalStrenght, defaultValue=<rgba 0.5 0.5 1 1>}.RGB
  )
  Multiply(
   A=ScalarParameter() {defaultValue=5, parameterName=DitherDepthThershold}
   B=MaterialFunctionCall() {materialFunction=&DitherTemporalAA}.Result
  )
  ScalarParameter() {defaultValue=0.3, parameterName=Specular}
  ScalarParameter() {defaultValue=0.4, parameterName=Roughness}
  ScalarParameter() {defaultValue=0.4, parameterName=Metalic}
  n13 = Multiply(A=VertexNormalWS(), B=ScalarParameter() {defaultValue=1, group=WPO, parameterName=Tendrill Thickness})
  n14 = DynamicParameter() {paramNames [Tendril Thickness, WPO Amount, Emissive, NA]}
  n15 = TextureObject() {samplerType=SAMPLERTYPE_Normal, texture=&T_Weap_Distort_N_02}
  n16 = ScalarParameter() {defaultValue=10, parameterName=WPOScale}
  n17 = WorldPosition()
  n18 = ScalarParameter() {defaultValue=1, group=WPO, parameterName=WPO Pan Speed V}
  n19 = Multiply(A=Time(), B=@n18)
  n20 = ComponentMask(None=MaterialFunctionCall!(
    XY Texture=@n15
    YZ Texture=@n15
    XZ Texture=@n15
    XY Scale=@n16
    YZ Scale=@n16
    XZ Scale=@n16
    Blend Exponent Y Z
    Blend Mult Y Z
    Blend Exponent X
    Blend Mult X
    Single Texture
    Normal
    WorldPosition=Add(A=@n17.XYZ, B=@n19)
   ) {materialFunction=/Engine/Functions/Engine_MaterialFunctions03/Texturing/WorldCoordinate3Way}.XYZ Output) {g=true}
  n21 = TextureCoordinate()
  n22 = Multiply(A=ComponentMask(None=@n21) {r=true}, B=ScalarParameter() {defaultValue=1.5, group=WPO, parameterName=WPO Tile U})
  n23 = Multiply(A=ComponentMask(None=@n21) {g=true}, B=ScalarParameter() {defaultValue=3, group=WPO, parameterName=WPO Tile V})
  n24 = Time() {bOverride_Period=true, period=64}
  n25 = Time() {period=64}
  n26 = ScalarParameter() {group=NitroFlow_Options, parameterName=UseLinearTime}
  n27 = ScalarParameter() {group=WPO, parameterName=WPO Pan Speed U}
  n28 = TextureSample(UVs=Panner(Coordinate=AppendVector(A=@n22, B=@n23), Time=LinearInterpolate(A=@n24, B=@n25, Alpha=@n26), Speed=AppendVector(A=@n27, B=@n18)) {speedX=-1, speedY=0.1}) {samplerType=SAMPLERTYPE_Normal, texture=&T_Weap_Distort_N_02}
  n29 = ScalarParameter() {group=NitroFlow_Options, parameterName=NitroFlow_UseModelUVs, sortPriority=0}
  n30 = Add(A=Multiply(A=LinearInterpolate(A=@n20, B=@n28.G, Alpha=@n29)) {constB=0.5}) {constB=0.4}
  n31 = ScalarParameter() {defaultValue=1, group=WPO, parameterName=WPO Normal Contrast}
  n32 = Multiply(
   A=Multiply(
    A=MaterialFunctionCall(In=@n30, Contrast=@n31) {materialFunction=&CheapContrast}.Result
    B=ScalarParameter() {defaultValue=1.5, group=WPO, parameterName=WPO Amnt}
   )
   B=DynamicParameter() {paramNames [Tendril Thickness, WPO Amount, Emissive, NA]}.WPO Amount
  )
  StaticSwitchParameter(True=Multiply(A=Multiply(A=@n13, B=@n14.Tendril Thickness), B=@n32) {constA=1}, False=@n32) {defaultValue=true, group=WPO, parameterName=Use Tendril Thickness}
 ]
}
