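# Master character material for the Vertigo skin, with static-switch variants for its pickaxe and backpack.
# Character PreFX feeds FX-masked thin-film and throbbing functions, then Fresnel metallic/roughness tweaks and base-color emissive boost before PostFX.
# The Pickaxe switch adds a reveal that collapses vertices toward the object position through world position offset, driven by a gradient-wipe mask.
# The Backpack switch adds vertex-normal displacement from panning electricity noise projected in local space.
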Material /Game/Characters/Player/Male/Large/Bodies/M_LRG_Vertigo/Materials/M_Vertigo_Master
graph {
 properties {
  bUseMaterialAttributes=true
  bUsedWithMeshDeformer=true
  bUsedWithMorphTargets=true
  bUsedWithSkeletalMesh=true
  pixelDepthOffsetMode=PDOM_Legacy
  usedShadingModels=Default Lit
 }
 nodes [
  n1 = ScalarParameter() {group=BackpackWPO, parameterName=WPO Offset}
  n2 = ComponentMask(None=MaterialFunctionCall!(
    TextureObject=TextureObject() {samplerType=SAMPLERTYPE_Masks, texture=/Game/Characters/Player/Male/Medium/Bodies/M_MED_Storm_Soldier/Textures/ElectricityNoise_M}
    TextureSize=ScalarParameter() {defaultValue=100, parameterName=TextureSize}
    WorldPosition=MaterialFunctionCall() {materialFunction=&LocalPosition}.Local Position
    Export Float 4
    In=ScalarParameter() {defaultValue=0.2, parameterName=PanningSpeed}
    World Space Normal
    ProjectionTransitionContrast
   ) {materialFunction=/Game/Athena/SafeZone/WorldAlignedTexture_ScalePan}.XYZ Texture) {g=true, r=true}
  Multiply(A=@n1, B=@n2)
  n3 = MaterialFunctionCall() {materialFunction=&BaseFN_CharacterFunction_PreFX}
  n4 = TextureSampleParameter2D() {
   parameterName=FX
   samplerType=SAMPLERTYPE_Masks
   texture=/Game/Characters/Player/Male/Large/Bodies/M_LRG_Vertigo/Textures/T_M_LRG_Vertigo_Body_FX
  }
  n5 = ChannelMaskParameter(None=@n4.RGB) {maskChannel=Blue, parameterName=ThinFilm Channel}
  n6 = GetMaterialAttributes(None=MaterialFunctionCall!(
    Alpha=ChannelMaskParameter(None=@n4.RGB) {parameterName=Throbbing Channel}.R
    Pre SKinned Local=@n3.Pre skinned Local Position
    ThicknessMap
    In=MaterialFunctionCall!(Alpha=@n5.R, Material Attributes=@n3.Result) {materialFunction=/Game/Weapons/WeaponSkins/MaterialFunction/ThinFilmFunction}.Result
   ) {materialFunction=&/MF_Throbbing}.Result) {attributeGetTypes [
    #1
    #2
    #3
    #4
   ]}
  n7 = LinearInterpolate(A=@n6.Metallic, B=Constant() {r=0.6}, Alpha=@n3.Fresnel)
  n8 = ScalarParameter() {defaultValue=1, group=Vertigo, parameterName=vertigo_MetallicMult}
  n9 = Multiply(A=@n6.Roughness) {constB=0.75}
  n10 = Multiply(A=@n6.Base Color, B=ScalarParameter() {group=Vertigo, parameterName=vertigoBaseColEmissive})
  n11 = MaterialFunctionCall(
   AO=@n3.AO
   Distance Blend=@n3.AO
   Fresnel=@n3.AO
   Gmap=@n3.AO
   Pre Skinned Local Position=@n3.AO
   Material=SetMaterialAttributes!(
    MaterialAttributes=@n6.MaterialAttributes
    Metallic=Multiply(A=@n7, B=@n8)
    Roughness=LinearInterpolate(A=@n6.Roughness, B=@n9, Alpha=@n3.Fresnel)
    Emissive Color=Add(A=@n6.Emissive Color, B=@n10)
   ) {attributeSetTypes [#1, #2, #3]}
  ) {materialFunction=&BaseFN_CharacterFunction_PostFX}
  n12 = GetMaterialAttributes(None=@n11.Result) {attributeGetTypes [#5]}
  n13 = Multiply(A=Subtract(A=WorldPosition().XYZ, B=ObjectPositionWS())) {constB=-1}
  n14 = ChannelMaskParameter(None=@n4.RGBA) {parameterName=GradientWipeChannel}
  n15 = ScalarParameter() {defaultValue=1, parameterName=RevealPickaxe}
  n16 = Saturate(None=Add(A=Subtract(A=@n14.R), B=LinearInterpolate(Alpha=@n15) {constA=1.8, constB=0}))
  n17 = StaticSwitchParameter(
   True=SetMaterialAttributes!(MaterialAttributes=@n12.MaterialAttributes, World Position Offset=LinearInterpolate(A=@n12.World Position Offset, B=@n13, Alpha=@n16)) {attributeSetTypes [#5]}
   False=@n11.Result
  ) {parameterName=Pickaxe}
  n18 = GetMaterialAttributes(None=@n17) {attributeGetTypes [#5]}
  n19 = Multiply(A=ConstantBiasScale(None=@n2) {bias=-0.5, scale=2}, B=ScalarParameter() {defaultValue=5, group=BackpackWPO, parameterName=WPO Intensity})
  n20 = Multiply(A=ComponentMask(None=Add(A=@n19, B=@n1)) {g=true}, B=VertexNormalWS())
  StaticSwitchParameter(
   True=SetMaterialAttributes(MaterialAttributes=@n18.MaterialAttributes, World Position Offset=Add(A=@n18.World Position Offset, B=@n20)) {attributeSetTypes [#5]}
   False=@n17
  ) {parameterName=Backpack}
 ]
}
