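# Control Rig for the tongue face accessory on the large male Vertigo player body, adding procedural motion to the tongue bone chain.
# Chain harmonics from tongue_1 apply a noisy wave, with pendulum and reach disabled.
# The tongue_6 tip is then clamped to a cylindrical range around spine_05, and tongue_3 rotates by the resulting correction, propagating to its children.
# Debug drawing is on for both the harmonics and the clamp.
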ControlRigBlueprint /Game/Characters/Player/Male/Large/Bodies/M_LRG_Vertigo/Meshes/Parts/M_LRG_Vertigo_Tongue_FaceAcc_CtrlRig
graphs [
 RigVMFunctionLibrary {links [], nodes []}
 RigVMModel {nodes [
   RigUnit_BeginExecution = &RigUnit_BeginExecution()
   /Script/ControlRig.RigUnit_ChainHarmonics(
    ChainRoot=tongue_1
    DrawWorldOffset=(Rotation=(X=0,Y=0,Z=0,W=1),Translation=(X=0,Y=0,Z=0),Scale3D=(X=1,Y=1,Z=1))
    Pendulum=(bEnabled=False,PendulumStiffness=2,PendulumGravity=(X=0,Y=0,Z=0),PendulumBlend=0.75,PendulumDrag=0.98,PendulumMinimum=0,PendulumMaximum=0.1,PendulumEase=Linear,UnwindAxis=(X=0,Y=1,Z=0),UnwindMinimum=0.2,UnwindMaximum=0.05)
    Reach=(bEnabled=False,ReachTarget=(X=0,Y=0,Z=0),ReachAxis=(X=1,Y=0,Z=0),ReachMinimum=0,ReachMaximum=0,ReachEase=SineEaseInOut)
    Speed=(X=0.25,Y=0.25,Z=0.1)
    Wave=(bEnabled=True,WaveFrequency=(X=0.2,Y=0.2,Z=0.2),WaveAmplitude=(X=0.25,Y=2.5,Z=1.5),WaveOffset=(X=1,Y=0,Z=2),WaveNoise=(X=8,Y=8,Z=30),WaveMinimum=0,WaveMaximum=0.75,WaveEase=SineEaseIn)
    WaveCurve=(EditorCurveData=(Keys=((Time=-0.0198,Value=0.4929),(Time=0.9855,Value=0.5049)),PreInfinityExtrap=RCCE_Constant,PostInfinityExtrap=RCCE_Constant,DefaultValue=340282346638528859811704183484516925440))
    bDrawDebug=True
    ExecutePin=@RigUnit_BeginExecution.ExecutePin
   ) {is=struct RigVMUnitNode}
   n1 = &RigUnit_GetBoneTransform(Bone=tongue_6)
   n2 = &RigUnit_GetBoneTransform(Bone=tongue_3)
   RigUnit_GetBoneTransform = &RigUnit_GetBoneTransform(Bone=spine_05)
   n3 = /Script/RigVM.RigVMFunction_MathTransformClampSpatially(
    Axis=X
    DebugColor=(R=1,G=0,B=0,A=1)
    DebugThickness=1
    Maximum=200
    Minimum=20
    Type=Cylinder
    bDrawDebug=true
    Space=@RigUnit_GetBoneTransform.Transform
    Value=@n1.Transform
   ) {
    is=struct RigVMUnitNode
    template="ClampSpatially::Execute(in Value,in Axis,in Type,in Minimum,in Maximum,in Space,in bDrawDebug,in DebugColor,in DebugThickness,out Result)"
   }
   > /Script/ControlRig.RigUnit_AddBoneTransform(
    Bone=tongue_3
    Transform=(Rotation=(X=0,Y=0,Z=0,W=1),Translation=(X=0,Y=0,Z=0),Scale3D=(X=1,Y=1,Z=1))
    Weight=1
    bPropagateToChildren=True
    Transform.Rotation=/Script/RigVM.RigVMFunction_MathQuaternionFromTwoVectors(
     A=&RigVMFunction_MathTransformMakeRelative(Global=@n1.Transform, Parent=@n2.Transform).Local.Translation
     B=&RigVMFunction_MathTransformMakeRelative(Global=@n3.Result, Parent=@n2.Transform).Local.Translation
    ) {is=struct RigVMUnitNode, template="FromTwoVectors::Execute(in A,in B,out Result)"}.Result
   ) {is=struct RigVMUnitNode}
  ]}
]
hierarchy [
 root {loc=<0 0 0>, rot=<0 0 0 1>}
 pelvis {parent=root, loc=<0 -0.7321 92.4218>, rot=<0.0127 0.707 -0.0127 -0.707>}
 spine_01 {parent=pelvis, loc=<5.6002 -0.5128 0>, rot=<0 0 0.034 -0.9994>}
 spine_02 {parent=spine_01, loc=<8.0753 0.0139 0>, rot=<0 0 0.0238 -0.9997>}
 spine_03 {parent=spine_02, loc=<7.0393 0 0>, rot=<0 0 0.0149 -0.9999>}
 spine_04 {parent=spine_03, loc=<8.1023 0 0>, rot=<0 0 -0.0132 -0.9999>}
 spine_05 {parent=spine_04, loc=<10.0146 0 0>, rot=<0 0 -0.0015 -1>}
 neck_01 {parent=spine_05, loc=<10.8149 1.9196 0>, rot=<0 0 -0.1864 -0.9825>}
 neck_02 {parent=neck_01, loc=<4.7347 0.3045 0>, rot=<0 0 0.0363 -0.9993>}
 head {parent=neck_02, loc=<5.7632 0.5552 0>, rot=<0 0 0.115 -0.9934>}
 faceAttach {parent=head, loc=<0.6419 -1.7133 0>, rot=<0.0082 -0.7071 -0.0082 -0.7071>, scale=<0.89 0.89 0.89>}
 C_jaw {parent=faceAttach, loc=<0 -1.3446 3.1231>, rot=<0 0 0 1>}
 tongue {parent=C_jaw, loc=<0 5.1808 -4.4676>, rot=<-0.1568 0 0 -0.9876>}
 tongue_base {parent=head, loc=<0.0313 -2.4747 0>, rot=<0 0 -0.761 -0.6488>}
 tongue_1 {parent=tongue_base, loc=<0.2 0 0>, rot=<0 0 0 -1>}
 tongue_2 {parent=tongue_1, loc=<4 0 0>, rot=<0 0 -0.098 -0.9952>}
 tongue_3 {parent=tongue_2, loc=<6.2543 0 0>, rot=<0 0 -0.0235 -0.9997>}
 tongue_4 {parent=tongue_3, loc=<7.1497 0 0>, rot=<0 0 -0.0149 -0.9999>}
 tongue_5 {parent=tongue_4, loc=<6.2646 0 0>, rot=<0 0 0.0072 -1>, scale=<0.9998 0.9999 0.9999>}
 tongue_6 {parent=tongue_5, loc=<4.6966 0 0>, rot=<0 0 0 -1>, scale=<1.0025 1.0011 1.0014>}
 offset_neck_01 {parent=spine_05, loc=<13.8043 -0.5479 0>, rot=<0 0 -0.1864 -0.9825>}
 offset_neck_02 {parent=offset_neck_01, loc=<6.1557 0.9858 0>, rot=<0 0 0.0363 -0.9993>}
 offset_head {parent=offset_neck_02, loc=<5.4291 0.8903 0>, rot=<0 0 0.115 -0.9934>}
 top_jaw {parent=offset_head, loc=<1.612 -2.3994 0>, rot=<0.4979 0.5021 0.5021 0.4979>}
 bot_jaw {parent=offset_head, loc=<1.612 -2.3994 0>, rot=<-0.4979 -0.5021 -0.5021 -0.4979>}
]
